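# Physics asset for the Mango T3 vehicle body skeletal mesh, part of the vehicle cosmetics set.
# It holds a single body on the root bone with no collision shapes listed.
# It defines no constraints, so the mesh has no jointed or ragdoll simulation between bones.
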PhysicsAsset /VehicleCosmetics/Bodies/Mango/Mango_T3/Meshes/SK_Body_MangoT3_Physics
bodies {root []}
constraints []
